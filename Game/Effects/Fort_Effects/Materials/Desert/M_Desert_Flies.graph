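# Unlit translucent particle sprite material, named for desert flies, that draws each particle as a procedural dot without textures.
# Particle color feeds a time-of-day function, tinted bluer for evening and night, then an emissive adjustment function.
# Opacity multiplies particle alpha by a centered UV sphere mask with adjustable radius and hardness.
# MF_SmoothLod scales that opacity with parameterized near and far distance fades.
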Material /Game/Effects/Fort_Effects/Materials/Desert/M_Desert_Flies
graph {
 properties {
  bUsedWithParticleSprites=true
  blendMode=BLEND_Translucent
  pixelDepthOffsetMode=PDOM_Legacy
  shadingModel=MSM_Unlit
  usedShadingModels=Unlit
 }
 nodes [
  n1 = ParticleColor()
  n2 = Constant3Vector() {constant=<rgba 0.4 0.6 1 0>}
  n3 = Constant3Vector() {constant=<rgba 0.2 0.2 1 0>}
  MaterialFunctionCall(Emissive=MaterialFunctionCall(Color_Morning=@n1.RGB, Color_Day=@n1.RGB, Color_Evening=Multiply(A=@n1.RGB, B=@n2), Color_Night=Multiply(A=@n1.RGB, B=@n3)) {materialFunction=&Color_Time_of_Day}.Result) {materialFunction=&MF_FX_EmissiveAdjustments}
  n4 = TextureCoordinate()
  n5 = Constant2Vector() {g=0.5, r=0.5}
  n6 = ScalarParameter() {defaultValue=0.5, parameterName=Radius}
  n7 = ScalarParameter() {parameterName=Hardness}
  Multiply(
   A=Multiply(A=@n1.A, B=SphereMask(A=@n4, B=@n5, Radius=@n6, Hardness=@n7))
   B=MaterialFunctionCall!(
    NearFadeLength=ScalarParameter() {defaultValue=512, parameterName="NearFadeLength (S)"}
    NearFadeOffset=ScalarParameter() {defaultValue=64, parameterName="NearFadeOffset (S)"}
    NearFadeCurve
    NearFadeStartingScale
    NearFadeEndingScale=ScalarParameter() {defaultValue=-0.5, parameterName="NearFadeEndingScale (S)"}
    FarFadeLength=ScalarParameter() {defaultValue=2048, parameterName="FarFadeLength (S)"}
    FarFadeOffset=ScalarParameter() {defaultValue=1024, parameterName="FarFadeOffset (S)"}
    FarFadeCurve
    FarFadeEndingScale=ScalarParameter() {defaultValue=3, parameterName="FarFadeEndingScale (S)"}
    Opacity Near Distance Bias=ScalarParameter() {defaultValue=1, parameterName="Opacity Near Distance Bias (S)"}
    Opacity Far Distance Bias=ScalarParameter() {defaultValue=0.25, parameterName="Opacity Far Distance Bias (S)"}
   ) {materialFunction=/Game/Effects/Fort_Effects/Materials/_Functions/MF_SmoothLod}.Opacity
  )
 ]
}
