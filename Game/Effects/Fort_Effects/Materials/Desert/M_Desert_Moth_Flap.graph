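# Translucent, two-sided, lit mesh-particle material for a flapping desert moth effect.
# Desert creatures texture times particle color feeds an emissive adjustment function, with both alphas multiplied for opacity.
# A time sine, scaled by vertex color red, builds an offset along vertex normals for the flap.
# Dynamic parameters give each particle its flap phase offset, range and speed, so animation needs no skeleton.
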Material /Game/Effects/Fort_Effects/Materials/Desert/M_Desert_Moth_Flap
graph {
 properties {
  bEnableResponsiveAA=true
  bUsedWithMeshParticles=true
  blendMode=BLEND_Translucent
  pixelDepthOffsetMode=PDOM_Legacy
  translucencyLightingMode=TLM_VolumetricPerVertexDirectional
  twoSided=true
  usedShadingModels=Default Lit
 }
 nodes [
  n1 = TextureSample() {texture=/Game/Effects/Fort_Effects/Effects/Environment/Desert/Dust/Organize/T_DesertCreatures_D}
  n2 = ParticleColor()
  MaterialFunctionCall(Emissive=Multiply(A=@n1.RGB, B=@n2.RGB)) {materialFunction=&MF_FX_EmissiveAdjustments}
  n3 = Time()
  n4 = DynamicParameter() {defaultValue=<rgba 0 8 1 1>, paramNames [FlapRandomOffset, ZFlappingRange, FlapSpeed, Param4]}
  n5 = Sine(None=Multiply(A=Add(A=@n3, B=@n4.FlapRandomOffset), B=@n4.FlapSpeed))
  Multiply(A=VertexNormalWS(), B=Multiply(A=VertexColor().R, B=Multiply(A=@n5, B=@n4.ZFlappingRange)))
  Multiply(A=@n1.A, B=@n2.A)
  VectorParameter() {parameterName=SubSurfColor, defaultValue=<rgba 0.69 0.5743 0.414 1>}
 ]
}
