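# Translucent, two-sided lit sprite material for desert trash particles, usable with Cascade and Niagara sprites.
# World position offset tumbles each sprite around a random axis at a random speed, plus the near/far distance offset from MF_SmoothLod.
# A packed smoke-trail texture with randomized tiling supplies opacity from its blue channel, scaled by particle alpha and distance fade.
# Particle color gets slight random darkening, then feeds a dim, adjusted emissive.
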Material /Game/Effects/Fort_Effects/Materials/Desert/M_Desert_Trash_S_GPU
graph {
 properties {
  bUsedWithNiagaraSprites=true
  bUsedWithParticleSprites=true
  blendMode=BLEND_Translucent
  pixelDepthOffsetMode=PDOM_Legacy
  translucencyLightingMode=TLM_VolumetricPerVertexNonDirectional
  twoSided=true
  usedShadingModels=Default Lit
 }
 nodes [
  n1 = Constant3Vector() {constant=<rgba 0 1 0.6 0>}
  n2 = Constant3Vector() {constant=<rgba 1 0.25 0 0>}
  n3 = ParticleRandom()
  n4 = Time()
  n5 = ParticleRandom()
  n6 = MaterialFunctionCall!(
   NearFadeLength=ScalarParameter() {defaultValue=512, parameterName="NearFadeLength (S)"}
   NearFadeOffset=ScalarParameter() {defaultValue=64, parameterName="NearFadeOffset (S)"}
   NearFadeCurve
   NearFadeStartingScale
   NearFadeEndingScale=ScalarParameter() {defaultValue=-0.5, parameterName="NearFadeEndingScale (S)"}
   FarFadeLength=ScalarParameter() {defaultValue=2048, parameterName="FarFadeLength (S)"}
   FarFadeOffset=ScalarParameter() {defaultValue=1024, parameterName="FarFadeOffset (S)"}
   FarFadeCurve
   FarFadeEndingScale=ScalarParameter() {defaultValue=1.25, parameterName="FarFadeEndingScale (S)"}
   Opacity Near Distance Bias=ScalarParameter() {defaultValue=1, parameterName="Opacity Near Distance Bias (S)"}
   Opacity Far Distance Bias=ScalarParameter() {defaultValue=0.25, parameterName="Opacity Far Distance Bias (S)"}
  ) {materialFunction=/Game/Effects/Fort_Effects/Materials/_Functions/MF_SmoothLod}
  Add(
   A=RotateAboutAxis(
    NormalizedRotationAxis=LinearInterpolate(A=@n1, B=@n2, Alpha=@n3)
    RotationAngle=Multiply(A=@n4, B=@n5)
    PivotPoint=ParticlePositionWS()
    Position=WorldPosition().XYZ
   )
   B=@n6.WPO
  )
  n7 = ParticleColor()
  n8 = Multiply(B=@n7.RGB) {constA=0.8}
  n9 = ParticleRandom()
  MaterialFunctionCall(Emissive=Multiply(A=LinearInterpolate(A=@n8, B=@n7.RGB, Alpha=@n9), B=ScalarParameter() {defaultValue=0.05, parameterName=MultEmissive})) {materialFunction=&MF_FX_EmissiveAdjustments}
  n10 = TextureSample(UVs=LinearInterpolate(A=TextureCoordinate() {uTiling=0.25}, B=TextureCoordinate(), Alpha=ParticleRandom())) {texture=/Game/Effects/Fort_Effects/Textures/Smoke/T_Smoke_ThinTrail_Packed}
  Multiply(A=Multiply(A=@n7.A, B=Add(A=@n10.B, B=@n10.B)), B=@n6.Opacity)
 ]
}
