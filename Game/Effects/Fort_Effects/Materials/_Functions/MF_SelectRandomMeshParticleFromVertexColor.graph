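# Material function that shows one of several models built into a single static mesh, intended for randomizing mesh particles via a vertex offset.
# Select Mesh and a vertex color channel, green by default, are each scaled by Model Count and rounded up, then compared.
# Matching vertices get zero offset; the rest get negated local position, collapsing unselected models to the pivot.
# The Offset In World input is declared but unconnected.
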MaterialFunction /Game/Effects/Fort_Effects/Materials/_Functions/MF_SelectRandomMeshParticleFromVertexColor
graph {nodes [
  FunctionInput() {bUsePreviewValueAsDefault=true, inputName=Offset In World, sortPriority=2, previewValue=<-5000 0 0 1>}
  n1 = FunctionInput() {description="0 to 1, selects mesh based on their green channel vertex color 0-255", inputName=Select Mesh, inputType=FunctionInput_Scalar}
  n2 = FunctionInput() {
   bUsePreviewValueAsDefault=true
   description="How many models are built into the static mesh actor, each has an ascending vertex color divided by number of objects"
   inputName=Model Count
   inputType=FunctionInput_Scalar
   sortPriority=1
   previewValue=<4 0 0 1>
  }
  n3 = Multiply(A=Clamp(None=@n1) {minDefault=0}, B=@n2)
  n4 = FunctionInput(Preview=VertexColor().G) {
   bUsePreviewValueAsDefault=true
   description=Defaults to Green
   inputName="Vertex Color Channel Override "
   inputType=FunctionInput_Scalar
   sortPriority=3
  }
  n5 = Multiply(A=@n4, B=@n2) {constB=4}
  n6 = Multiply(
   A=MaterialFunctionCall() {materialFunction=&LocalPosition}."Local Position (Excluding Offsets)"
   B=Constant() {r=-1}
  )
  FunctionOutput(None=If(A=Ceil(None=@n3), B=Ceil(None=@n5), "A > B"=@n6, "A == B"=Constant(), "A < B"=@n6))
 ]}
