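# Masked, two-sided lit mesh-particle material for small desert trash debris, using the trash prop color and normal textures.
# A dynamic parameter picks one of several models through a vertex-color selection function, whose result is added to animated vector noise as an offset.
# The noise samples actor-relative world position scrolled over time, with per-particle curl scale, intensity and speed.
# Particle alpha feeds temporal-AA dither, and particle color feeds an emissive-adjustment function.
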Material /Game/Effects/Fort_Effects/Materials/Desert/M_Desert_Trash_S_Mesh
graph {
 properties {bUsedWithMeshParticles=true, blendMode=BLEND_Masked, pixelDepthOffsetMode=PDOM_Legacy, twoSided=true, usedShadingModels=Default Lit}
 nodes [
  n1 = WorldPosition()
  n2 = ActorPositionWS()
  n3 = DynamicParameter() {paramNames [PickMesh, Curl_Scale, Curl_Intensity, Curl_Speed]}
  n4 = Time()
  n5 = Divide(
   A=Subtract(A=Subtract(A=@n1.XYZ, B=@n2), B=Multiply(A=@n3.Curl_Speed, B=@n4))
   B=DynamicParameter() {paramNames [PickMesh, Curl_Scale, Curl_Intensity, Curl_Speed]}.Curl_Scale
  )
  Add(
   A=MaterialFunctionCall!(
    Select Mesh=DynamicParameter() {paramNames [PickMesh, Curl_Scale, Curl_Intensity, Curl_Speed]}.PickMesh
    Model Count=ScalarParameter() {defaultValue=5, parameterName="Model Count (S)"}
    Offset In World
    "Vertex Color Channel Override "
   ) {materialFunction=/Game/Effects/Fort_Effects/Materials/_Functions/MF_SelectRandomMeshParticleFromVertexColor}.Result
   B=Multiply(
    A=DynamicParameter() {paramNames [PickMesh, Curl_Scale, Curl_Intensity, Curl_Speed]}.Curl_Intensity
    B=VectorNoise(World Position=@n5) {noiseFunction=VNF_VectorALU}
   )
  )
  MaterialFunctionCall(Alpha Threshold=ParticleColor().A) {materialFunction=&DitherTemporalAA_Fortnite}
  MaterialFunctionCall(Emissive=ParticleColor().RGB) {materialFunction=&MF_FX_EmissiveAdjustments}
  ScalarParameter() {parameterName="Select Mesh (S)"}
  ScalarParameter() {parameterName=Specular}
  ScalarParameter() {defaultValue=1, parameterName=Roughness}
  TextureSample() {texture=/Game/Environments/Props/Trash/Textures/Prop_Trash_D}
  TextureSample() {samplerType=SAMPLERTYPE_Normal, texture=/Game/Environments/Props/Trash/Textures/Prop_Trash_N}
 ]
}
